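; Floor variant of the Athena poison dart trap, handling placement setup, trigger registration and cosmetic feedback.
; Placement caches the mesh's socket names, adds the trigger and plays dust and sound.
; The ActivateTrap GameplayCue drives a timestamp-based material jiggle with fire and delayed re-arm sounds; MuzzleFire spawns per-socket dart effects.
; Lifecycle events apply a friendly material for the local player's team in Athena, or play destruction effects when durability runs out.
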(event EventOnPlaced
  ;[1 1 1] Initial Setup
  (Development|PrintString (Utilities|String|ToString(Boolean) (Utilities|Enum|Equal(Enum) (Fortnite|GetSubGame) "Athena")) true true "(R=1.000000,G=0.010197,B=0.030093,A=1.000000)")
  (Utilities|FlowControl|DoOnce false
    (:Completed
      (Utilities|FlowControl|Sequence
        (:then_0
          (Variables|Default|SetMuzzleSockets (Transformation|GetAllSocketNames (Variables|Debug|GetStaticMeshComponent)))
          (Development|PrintString "Hello"))
        (:then_1
          (Trap|AddTriggerComponent (Variables|Trap_Floor_Spikes|GetTrigger))
          (if (not (Variables|Default|GetbPlayFewerSounds))
            (Audio|PlaySoundatLocation (Variables|Audio|GetPlace_Trap_Sound) (Transformation|GetActorLocation))
            (Utilities|TrapPlacementDustFX :Attach%20to%20Component (Variables|Debug|GetStaticMeshComponent) :Socket "PlacedFX" :Rotation "0, 0, 0" :Burst%20Rate 15 :Size%20%28Volume%29 "200.000000,200.000000,15.000000" :Velocity%20Scale "1,1,1" :Size%20Scale%20%28Particles%29 1)))))))

(event Custom|Jiggle
  (params (Vector :type Vector))
  ;[1 1 1] Initial Setup
  (Rendering|Material|SetVectorParameterValueonMaterials (Variables|Debug|GetStaticMeshComponent) "AmountVector" Vector))

(event EventOnFinishedBuilding
  ;[1 1 1] Finished building
  (Networking|CanExecuteCosmeticEvents
    (:True
      (CallFunction|ChangetoFriendlyMaterial)
      (if (not (Variables|Default|GetbPlayFewerSounds))
        (Audio|PlaySoundatLocation (Variables|Audio|GetTrap_Active_Sound) (Transformation|GetActorLocation))))))

(event EventOnWorldReady
  ;[1 1 1] Finished building
  (Networking|CanExecuteCosmeticEvents
    (:True
      (CallFunction|ChangetoFriendlyMaterial)
      (if (not (Variables|Default|GetbPlayFewerSounds))
        (Audio|PlaySoundatLocation (Variables|Audio|GetTrap_Active_Sound) (Transformation|GetActorLocation))))))

(event GameplayCue|GameplayCue.Abilities.Activation.Traps.ActivateTrap
  (params (EventType :type EGameplayCueEvent%20Enum :default "OnActive") (Parameters :type Gameplay%20Cue%20Parameters%20Structure%20%28by%20ref%29))
  (bind _returnvalue (Utilities|Time|GetGameTimeinSeconds))
  (bind _staticmeshcomponent (Variables|Debug|GetStaticMeshComponent))
  ;[1 1 1] On Fire
  (switch Utilities|FlowControl|Switch|SwitchonEGameplayCueEvent EventType
    (:Executed
      (Audio|PlaySoundatLocation (Variables|Audio|GetTrap_Fire_Sound) (Transformation|GetActorLocation))
      (Rendering|Material|SetScalarParameterValueonMaterials _staticmeshcomponent "TimestampA" _returnvalue)
      (Rendering|Material|SetScalarParameterValueonMaterials _staticmeshcomponent "TimestampB" (+ _returnvalue (Variables|Default|GetTimespan%20of%20Jiggle)))
      (CallFunction|Jiggle (Variables|Default|GetVector%20of%20Jiggle))
      (Utilities|FlowControl|Delay 3.5)
      (Audio|PlaySoundatLocation (Variables|Audio|GetTrap_Active_Sound) (Transformation|GetActorLocation)))))

(event Custom|MuzzleFire
  (params (Socket%20Name :type Name))
  ;[1 1 1] On Fire
  (Networking|CanExecuteCosmeticEvents
    (:True
      (Effects|SpawnEmitterAttached "/Game/Athena/Items/Traps/PoisonDartTrap/FX/P_BlowDart_Muzzle.P_BlowDart_Muzzle" (Variables|Debug|GetStaticMeshComponent) Socket%20Name (Transformation|GetSocketLocation (Variables|Debug|GetStaticMeshComponent) Socket%20Name) (Transformation|GetSocketRotation (Variables|Debug|GetStaticMeshComponent) Socket%20Name) "1.000000,1.000000,1.000000" "KeepWorldPosition"))))

(event EventOnReloadEnd
  ;[1 1 1] Reload End
  (Audio|PlaySoundatLocation (Variables|Audio|GetTrap_Active_Sound) (Transformation|GetActorLocation))
  (CallFunction|ChangetoFriendlyMaterial))

(event EventOnOutOfDurability
  ;[1 1 1] Durability Change
  (Utilities|FlowControl|Delay 1)
  (Audio|PlaySoundatLocation (Variables|Audio|GetTrap_Explode_Sound) (Transformation|GetActorLocation))
  (Effects|SpawnEmitteratLocation "/Game/Weapons/FORT_TRAPS/Spike_Trap/P_Spike_Trap_Destruction.P_Spike_Trap_Destruction" (Transformation|GetWorldLocation (Variables|Transformation|GetRootComponent))))

(event EventOnInitTeam
  ;[1 1 1] Detect if local player is on the same team as this trap
  (CallFunction|ChangetoFriendlyMaterial))

(fn ChangeToFriendlyMaterial
  (bind _returnvalue (Fortnite|GetLocalFortPlayerControllers true))
  (if (and (Utilities|Enum|Equal(Enum) (Fortnite|GetSubGame) "Athena") (> (Utilities|Array|Length _returnvalue) 0))
    (if (Math|Byte|Equal(Byte) (Variables|Team|GetTeamIndex) (Team|GetActorTeam (Utilities|Array|Get(acopy) _returnvalue)))
      (Rendering|Material|SetMaterial (Variables|Debug|GetStaticMeshComponent) 0 (Variables|Default|GetFriendlyTrapMaterial)))))

(fn ChangeToNormalMaterial
  (Rendering|Material|SetMaterial (Variables|Debug|GetStaticMeshComponent) 0 (Variables|Default|GetNormalTrapMaterial)))

(fn ConstructionScript)
