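# Translucent, two-sided lit mesh-particle material that draws the smoky tracer trail for the Poison Dart trap, with a dark yellow-green default color.
# SplineThicken sets trail width, plus an upward offset graded along V, with dynamic parameters driving size and offset.
# Opacity multiplies particle alpha, V-gradient fades, a sine edge mask across U and two summed panning smoke textures.
# Particle color feeds a scaled emissive contribution.
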Material /Game/Athena/Items/Traps/PoisonDartTrap/FX/M_DartTracer
graph {
 properties {
  bAutomaticallySetUsageInEditor=false
  bUsedWithMeshParticles=true
  bUsedWithStaticMesh=false
  blendMode=BLEND_Translucent
  pixelDepthOffsetMode=PDOM_Legacy
  translucencyLightingMode=TLM_VolumetricPerVertexNonDirectional
  twoSided=true
  usedShadingModels=Default Lit
 }
 nodes [
  n1 = ScalarParameter() {defaultValue=1, parameterName=WidthBase}
  n2 = DynamicParameter() {defaultValue=<rgba 1 1 1 10>, paramNames [VScale, WPO_Offset, Gradient Fade, SplineSizeScale]}
  n3 = ScalarParameter() {defaultValue=4, parameterName=WidthTip}
  n4 = Multiply(A=Constant3Vector() {constant=<rgba 0 0 1 0>}, B=ScalarParameter() {defaultValue=128, parameterName=Max WPO Up})
  n5 = TextureCoordinate()
  n6 = Power(Base=OneMinus(None=ComponentMask(None=@n5) {g=true}), Exp=ScalarParameter() {defaultValue=2, parameterName=WPOGradient})
  Add(
   A=MaterialFunctionCall(WidthBase=Multiply(A=@n1, B=@n2.SplineSizeScale), WidthTip=Multiply(A=@n3, B=@n2.SplineSizeScale)) {materialFunction=&SplineThicken}.WorldPositionOffset
   B=LinearInterpolate(
    B=Multiply(A=@n4, B=@n6)
    Alpha=DynamicParameter() {defaultValue=<rgba 1 1 1 10>, paramNames [VScale, WPO_Offset, Gradient Fade, SplineSizeScale]}.WPO_Offset
   ) {constB=0}
  )
  n7 = ParticleColor()
  MaterialFunctionCall(Emissive=Multiply(A=ScalarParameter() {defaultValue=0.1, parameterName=EmissiveContribution}, B=@n7.RGB)) {materialFunction=&MF_FX_EmissiveAdjustments}
  n8 = ComponentMask(None=TextureCoordinate()) {g=true}
  n9 = Power(Base=OneMinus(None=@n8), Exp=ScalarParameter() {defaultValue=1})
  n10 = Power(
   Base=@n8
   Exp=DynamicParameter() {defaultValue=<rgba 1 1 1 10>, paramNames [VScale, WPO_Offset, Gradient Fade, SplineSizeScale]}.Gradient Fade
  )
  n11 = Sine(None=ComponentMask(None=TextureCoordinate()) {r=true}) {period=2}
  n12 = Clamp(None=LinearInterpolate(Alpha=Multiply(A=@n11, B=@n11)) {constB=4})
  n13 = AppendVector(
   A=ScalarParameter() {defaultValue=1, parameterName=UScaleFactor}
   B=Multiply(
    A=DynamicParameter() {defaultValue=<rgba 1 1 1 10>, paramNames [VScale, WPO_Offset, Gradient Fade, SplineSizeScale]}.VScale
    B=ScalarParameter() {defaultValue=0.1, parameterName=VScaleFactor}
   )
  )
  n14 = MaterialFunctionCall(XY=Multiply(A=TextureCoordinate(), B=@n13)) {materialFunction=&Swizzle}
  n15 = MaterialFunctionCall(XY=Multiply(A=TextureCoordinate() {vTiling=2}, B=@n13)) {materialFunction=&Swizzle}
  n16 = Add(
   A=TextureSampleParameter2D(UVs=Panner(Coordinate=@n14.YX) {speedX=0.1}) {parameterName=TrailTexture_GChannel, texture=/Game/Effects/Fort_Effects/Textures/Smoke/T_ShotGun_Trace_Packed}.G
   B=TextureSampleParameter2D(UVs=Panner(Coordinate=@n15.YX) {speedX=0.3333}) {parameterName=TrailTexture_BChannel, texture=/Game/Effects/Fort_Effects/Textures/Smoke/T_Smoke_ThinTrail_Packed}.B
  )
  Multiply(A=@n7.A, B=Multiply(A=Multiply(A=@n9, B=@n10), B=Multiply(A=@n12, B=@n16)))
  VectorParameter() {parameterName=BaseColor, defaultValue=<rgba 0.168 0.2384 0.0078 1>}
 ]
}
